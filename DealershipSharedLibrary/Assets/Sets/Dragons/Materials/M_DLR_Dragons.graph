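# Masked, lit material for skeletal dragon effects in the Dealership set, producing a glowing, noise-driven look tinted through painted-variant color data.
# A normal-mapped Fresnel multiplies panning packed noise, warped by soft noise via bump offset and UV offset.
# An X gradient from local bounds shapes the emissive, which is scaled by time of day.
# The same gradient and noise feed contrast and temporal dither functions for a dissolving fade.
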Material /DealershipSharedLibrary/Assets/Sets/Dragons/Materials/M_DLR_Dragons
graph {
 properties {
  bAutomaticallySetUsageInEditor=false
  bUsedWithMeshDeformer=true
  bUsedWithSkeletalMesh=true
  blendMode=BLEND_Masked
  usedShadingModels=Default Lit
 }
 nodes [
  Constant() {r=0.02}
  n1 = Constant() {r=1}
  n2 = Constant() {r=1}
  n3 = Constant() {r=1}
  n4 = Constant() {r=1}
  n5 = VectorParameter() {parameterName=Color, defaultValue=<rgba 0 0.3164 1 1>}
  n6 = MaterialFunctionCall!(Primary Color=AppendVector(A=@n5.RGB, B=@n5.RGB), Tertiary Color, "Secondary  Color") {materialFunction=/DealershipSharedLibrary/FX/Materials/MaterialFunctions/MF_DLR_PaintedVariantsData}
  n7 = ScalarParameter() {defaultValue=0.8, parameterName=Fresnel Exp}
  n8 = Transform(None=TextureSample() {samplerType=SAMPLERTYPE_Normal, texture=&/T_Dragons_N}.RGB)
  n9 = Power(Base=Multiply(A=Fresnel(ExponentIn=@n7, Normal=@n8) {baseReflectFraction=0}) {constB=1.4})
  n10 = ComponentMask(None=VertexInterpolator(VS=LocalPosition() {localOrigin=InstancePreSkinning}.XYZ).PS) {b=true, r=true}
  n11 = VectorParameter() {defaultValue=<rgba 0.002 0.002 2 0.112>}
  n12 = Panner(Coordinate=Multiply(A=@n10, B=AppendVector(A=@n11.R, B=@n11.R)), Speed=AppendVector(A=@n11.B, B=@n11.B)) {speedX=-1}
  n13 = VectorParameter() {parameterName=Param01, defaultValue=<rgba 0.003 0.003 3 -0.12>}
  n14 = AppendVector(A=@n13.R, B=@n13.R)
  n15 = TextureSample(UVs=Panner(Coordinate=Multiply(A=@n10, B=@n14), Speed=AppendVector(A=@n13.B, B=@n13.B)) {speedX=-1}) {texture=/Game/Effects/Fort_Effects/Emotes/UltraLight/Texture/T_SoftNoise_BC7}
  n16 = ScalarParameter() {defaultValue=0.2, parameterName=Param_4}
  n17 = ScalarParameter() {defaultValue=0.5, parameterName=Param_1}
  n18 = ComponentMask(None=@n15.RGB) {g=true, r=true}
  n19 = TextureSample(UVs=Add(A=BumpOffset(Coordinate=@n12, Height=@n15.B, HeightRatioInput=@n16), B=Multiply(A=@n17, B=@n18))) {samplerType=SAMPLERTYPE_Masks, texture=/Game/Effects/Fort_Effects/Textures/T_PackedNoise}
  n20 = Add(A=Multiply(A=Multiply(A=@n19.G, B=@n19.G), B=ScalarParameter() {defaultValue=1, parameterName=Param_5})) {constB=0.5}
  n21 = Multiply(A=Multiply(A=@n6.Primary RGB, B=Multiply(A=@n9, B=@n20) {constA=1}), B=ScalarParameter() {defaultValue=2, parameterName=Emissive})
  n22 = ComponentMask(None=Divide(A=LocalPosition().XYZ, B=PreSkinnedLocalBounds().Extents)) {r=true}
  n23 = Saturate(None=OneMinus(None=@n22))
  n24 = Multiply(A=@n23, B=@n23)
  n25 = Add(A=Multiply(A=@n24) {constB=2})
  MaterialFunctionCall(Emissive=Multiply(
    A=MaterialFunctionCall(Value_Morning=@n1, Value_Day=@n2, Value_Evening=@n3, Value_Night=@n4) {materialFunction=&Scalar_Time_of_Day}.Result
    B=Multiply(A=@n21, B=@n25)
   )) {materialFunction=&MF_FX_EmissiveAdjustments}
  n26 = Subtract(A=@n22) {constB=0.5}
  n27 = Add(A=ScalarParameter() {defaultValue=0.5, parameterName=FadeIn}, B=Multiply(A=@n19.G, B=ScalarParameter() {defaultValue=-0.2}))
  MaterialFunctionCall(Alpha Threshold=MaterialFunctionCall(In=Add(A=@n26, B=@n27), Contrast=ScalarParameter() {defaultValue=1, parameterName=Param_3}) {materialFunction=&CheapContrast}.Result) {materialFunction=&DitherTemporalAA_Fortnite}
  Multiply(A=@n24, B=@n24)
  Multiply(A=@n6.Primary RGB, B=ScalarParameter() {defaultValue=0.1, parameterName=Param_6})
  ScalarParameter() {parameterName=Param_2}
 ]
}
